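# Physics asset for the face accessory part of the male medium PoolSwirlHat cosmetic character, simulating a dangling string and front hair attached to the head.
# Capsule bodies cover a large head volume, small base anchors and the dynamic bones.
# A four-segment string chain and a single front-hair segment hang from their bases with narrow swing and near-locked twist limits.
# Both bases connect to the head through free angular constraints.
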PhysicsAsset /BRCosmetics/Characters/Player/Male/Medium/Bodies/M_MED_PoolSwirlHat/Meshes/Parts/M_MED_PoolSwirlHat_FaceAcc_Physics
bodies {
 base_hairsFront [{shapeType=Capsule, length=0.505, radius=0.505, shapeName=base_hairsFront_capsule, rotation=<rot 0 -90 0>}]
 base_string_00 [{shapeType=Capsule, length=0.505, radius=0.505, shapeName=base_string_00_capsule, rotation=<rot 0 -90 0>}]
 dyn_hairsFront [{shapeType=Capsule, length=2, radius=1, shapeName=dyn_hairsFront_capsule, center=<2 0 0>, rotation=<rot 90 0 0>}]
 dyn_string_00 [{
   shapeType=Capsule
   length=4.3369
   radius=1.6968
   shapeName=dyn_string_00_capsule
   center=<3.3735 0 -2.1362>
   rotation=<rot 86.9435 179.9898 179.9896>
  }]
 dyn_string_01 [{
   shapeType=Capsule
   length=4.0336
   radius=1.7036
   shapeName=dyn_string_01_capsule
   center=<4.6674 0 -1.5765>
   rotation=<rot 80.2636 179.9989 179.9987>
  }]
 dyn_string_02 [{shapeType=Capsule, length=1.2435, radius=1.7565, shapeName=dyn_string_02_capsule, center=<2 0 0>, rotation=<rot 90 0 0>}]
 dyn_string_03 [{shapeType=Capsule, length=1.21, radius=1.79, shapeName=dyn_string_03_capsule, center=<2 0 0>, rotation=<rot 90 0 0>}]
 head [{shapeType=Capsule, length=0.1106, radius=37.6703, shapeName=head_capsule, center=<-21.0839 -4.506 0>}]
}
constraints [
 {bone1Name=dyn_string_00, bone2Name=base_string_00, swing1LimitDegrees=15, swing2LimitDegrees=40, twistLimitDegrees=1}
 {bone1Name=dyn_string_01, bone2Name=dyn_string_00, swing1LimitDegrees=10, swing2LimitDegrees=40, twistLimitDegrees=1}
 {bone1Name=dyn_string_02, bone2Name=dyn_string_01, swing1LimitDegrees=5, swing2LimitDegrees=60, twistLimitDegrees=1}
 {bone1Name=dyn_string_03, bone2Name=dyn_string_02, swing1LimitDegrees=5, swing2LimitDegrees=60, twistLimitDegrees=1}
 {
  bone1Name=base_string_00
  bone2Name=head
  swing1LimitDegrees=45
  swing1Motion=Free
  swing2LimitDegrees=45
  swing2Motion=Free
  twistLimitDegrees=45
  twistMotion=Free
 }
 {
  bone1Name=base_hairsFront
  bone2Name=head
  swing1LimitDegrees=45
  swing1Motion=Free
  swing2LimitDegrees=45
  swing2Motion=Free
  twistLimitDegrees=45
  twistMotion=Free
 }
 {bone1Name=dyn_hairsFront, bone2Name=base_hairsFront, swing1LimitDegrees=10, swing2LimitDegrees=20, twistLimitDegrees=1}
]
